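# Physics asset for the collar of the Bad Egg medium male character body, covering five dynamic collar segments and the head.
# Box bodies shape the back, back-left, back-right, left and right segments, each constrained to its own small capsule MSTR anchor bone with limited swing and twist.
# The back segments allow wider swing than the side ones.
# An unconstrained head capsule provides a volume alongside the collar bodies.
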PhysicsAsset /BRCosmetics/Characters/Player/Male/Medium/Bodies/M_MED_Bad_Egg/Meshes/M_MED_Bad_Egg_Collar_Physics
bodies {
 dyn_collar_back [{shapeType=Box, extentX=15.125, extentY=15.1875, extentZ=3.875, center=<6.5 0 0>}]
 dyn_collar_back_MSTR [{shapeType=Capsule, length=0.505, radius=0.505}]
 dyn_collar_bk_l [{shapeType=Box, extentX=15.125, extentY=17.1875, extentZ=3.875, center=<6.5 0 0>, rotation=<rot 0 0 6.5943>}]
 dyn_collar_bk_l_MSTR [{shapeType=Capsule, length=0.505, radius=0.505}]
 dyn_collar_bk_r [{shapeType=Box, extentX=15.125, extentY=17.1875, extentZ=3.875, center=<6.5 0 0>, rotation=<rot 0 0 -5.981>}]
 dyn_collar_bk_r_MSTR [{shapeType=Capsule, length=0.505, radius=0.505}]
 dyn_collar_l [{shapeType=Box, extentX=9.8125, extentY=7.5, extentZ=2.8125, center=<4.3838 0 0>}]
 dyn_collar_l_MSTR [{shapeType=Capsule, length=0.505, radius=0.505}]
 dyn_collar_r [{shapeType=Box, extentX=9.8125, extentY=7.5, extentZ=2.8125, center=<4.3838 0 0>}]
 dyn_collar_r_MSTR [{shapeType=Capsule, length=0.505, radius=0.505}]
 head [{shapeType=Capsule, length=7.4375, radius=6.375, center=<6.831 -0.0565 -0.0001>, rotation=<rot 90 0 14.1423>}]
}
constraints [
 {bone1Name=dyn_collar_l, bone2Name=dyn_collar_l_MSTR, swing1LimitDegrees=45, swing2LimitDegrees=45, twistLimitDegrees=15}
 {bone1Name=dyn_collar_bk_l, bone2Name=dyn_collar_bk_l_MSTR, swing1LimitDegrees=80, swing2LimitDegrees=45, twistLimitDegrees=15}
 {bone1Name=dyn_collar_back, bone2Name=dyn_collar_back_MSTR, swing1LimitDegrees=80, swing2LimitDegrees=45, twistLimitDegrees=15}
 {bone1Name=dyn_collar_bk_r, bone2Name=dyn_collar_bk_r_MSTR, swing1LimitDegrees=80, swing2LimitDegrees=45, twistLimitDegrees=15}
 {bone1Name=dyn_collar_r, bone2Name=dyn_collar_r_MSTR, swing1LimitDegrees=45, swing2LimitDegrees=45, twistLimitDegrees=15}
]
