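# Post-copy Anim Blueprint for the Bad Egg character body that animates coat, collar and lapel over the attached parent's pose.
# Thigh, head and right-arm rotation deltas drive interpolated additive bone corrections as pseudo pose drivers.
# Two RigidBody nodes then simulate coat and collar, with coat gravity and force varying for swim sprinting and skydiving.
# Driving an Ostrich, Ferret or Octopus vehicle disables coat physics and shrinks lower coat bones.
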AnimBlueprint /BRCosmetics/Characters/Player/Male/Medium/Bodies/M_MED_Bad_Egg/Meshes/M_MED_Bad_Egg_AnimBP
target_skeleton=&/M_MED_Bad_Egg_Skeleton
graphs [
 AnimGraph {nodes [
   n1 = Get in vehicle scale alpha() -> (in vehicle scale alpha:double)
   n2 = "Transform (Modify) Bone - Bone: dyn_coat_fr_l_MSTR"(
    ComponentPose=Local To Component(LocalPose=Copy Pose From Mesh() {bUseAttachedParent=true})
    Alpha=Get thigh_rot_l()
    Translation="0,0,0"
    Rotation="3.870,-45.640,-3.944"
    Scale="1,1,1"
   ) {
    rotationMode=BMM_Additive
    rotationSpace=BCS_BoneSpace
    translationMode=BMM_Additive
    translationSpace=BCS_BoneSpace
    alphaBoolBlend {blendOption=Linear}
    alphaScaleBiasClamp {bInterpResult=true, interpSpeedDecreasing=15, interpSpeedIncreasing=15}
    boneToModify {boneName=dyn_coat_fr_l_MSTR}
    rotation=<rot 3.87 -45.6401 -3.944>
   }
   n3 = "Transform (Modify) Bone - Bone: dyn_coat_fr_r_MSTR"(ComponentPose=@n2, Alpha=Get thigh_rot_r(), Translation="0,0,0", Rotation="0,37.71,0", Scale="1,1,1") {
    rotationMode=BMM_Additive
    rotationSpace=BCS_BoneSpace
    translationMode=BMM_Additive
    translationSpace=BCS_BoneSpace
    alphaBoolBlend {blendOption=Linear}
    alphaScaleBiasClamp {bInterpResult=true, interpSpeedDecreasing=15, interpSpeedIncreasing=15}
    boneToModify {boneName=dyn_coat_fr_r_MSTR}
    rotation {yaw=37.71}
   }
   n4 = Get Head_Rot_Back() -> (Head_Rot_Back:double)
   n5 = "Transform (Modify) Bone - Bone: dyn_collar_back_MSTR"(ComponentPose=@n3, Alpha=@n4, Translation="-2.5,0.000,3", Rotation="11.111,0.000,0.000", Scale="1,1,1") {
    rotationMode=BMM_Additive
    rotationSpace=BCS_BoneSpace
    translationMode=BMM_Additive
    translationSpace=BCS_BoneSpace
    alphaBoolBlend {blendOption=Linear}
    alphaScaleBiasClamp {bInterpResult=true, interpSpeedDecreasing=15, interpSpeedIncreasing=15}
    boneToModify {boneName=dyn_collar_back_MSTR}
    rotation {pitch=11.111}
    translation {x=-2.5, z=3}
   }
   n6 = "Transform (Modify) Bone - Bone: dyn_collar_bk_r_MSTR"(ComponentPose=@n5, Alpha=@n4, Translation="-0.000,0.000,9.612", Rotation="0,0,0", Scale="1,1,1") {
    rotationMode=BMM_Additive
    rotationSpace=BCS_BoneSpace
    translationMode=BMM_Additive
    translationSpace=BCS_BoneSpace
    alphaBoolBlend {blendOption=Linear}
    alphaScaleBiasClamp {bInterpResult=true, interpSpeedDecreasing=15, interpSpeedIncreasing=15, scale=0.5}
    boneToModify {boneName=dyn_collar_bk_r_MSTR}
    translation {z=9.612}
   }
   n7 = "Transform (Modify) Bone - Bone: dyn_collar_bk_l_MSTR"(ComponentPose=@n6, Alpha=@n4, Translation="0.000,1.974,7.341", Rotation="0,0,0", Scale="1,1,1") {
    rotationMode=BMM_Additive
    rotationSpace=BCS_BoneSpace
    translationMode=BMM_Additive
    translationSpace=BCS_BoneSpace
    alphaBoolBlend {blendOption=Linear}
    alphaScaleBiasClamp {bInterpResult=true, interpSpeedDecreasing=15, interpSpeedIncreasing=15, scale=0.5}
    boneToModify {boneName=dyn_collar_bk_l_MSTR}
    translation {y=1.974, z=7.341}
   }
   n8 = "Transform (Modify) Bone - Bone: dyn_rapel_r_MSTR"(
    ComponentPose=@n7
    Alpha=Get arm_rot_fwd_r() -> (arm_rot_fwd_r:double)
    Translation="-1.1,0.034,-1.529"
    Rotation="85,18.782,17.929"
    Scale="0.25,0.25,0.25"
   ) {
    rotationMode=BMM_Additive
    rotationSpace=BCS_BoneSpace
    scaleMode=BMM_Additive
    scaleSpace=BCS_BoneSpace
    translationMode=BMM_Additive
    translationSpace=BCS_BoneSpace
    alphaBoolBlend {blendOption=Linear}
    alphaScaleBiasClamp {bInterpResult=true, interpSpeedDecreasing=20, interpSpeedIncreasing=20}
    boneToModify {boneName=dyn_rapel_r_MSTR}
    rotation=<rot 85 18.782 17.929>
    scale=<0.25 0.25 0.25>
    translation=<-1.1 0.034 -1.529>
   }
   n9 = "Transform (Modify) Bone - Bone: dyn_rapel_r_MSTR"(
    ComponentPose=@n8
    Alpha=Get arm_rot_high_r() -> (arm_rot_high_r:double)
    Translation="-0.55,0,-0.25"
    Rotation="30,0,0"
    Scale="0.25,0.25,0.25"
   ) {
    rotationMode=BMM_Additive
    rotationSpace=BCS_BoneSpace
    scaleMode=BMM_Additive
    scaleSpace=BCS_BoneSpace
    translationMode=BMM_Additive
    translationSpace=BCS_BoneSpace
    alphaBoolBlend {blendOption=Linear}
    alphaScaleBiasClamp {bInterpResult=true, interpSpeedDecreasing=20, interpSpeedIncreasing=20}
    boneToModify {boneName=dyn_rapel_r_MSTR}
    rotation {pitch=30}
    scale=<0.25 0.25 0.25>
    translation {x=-0.55, z=-0.25}
   }
   n10 = "Transform (Modify) Bone - Bone: dyn_coat_l_02"(ComponentPose=@n9, Alpha=@n1, Translation="0,0,0", Rotation="0,0,0", Scale="0.1,0.1,0.1") {scaleMode=BMM_Replace, alphaBoolBlend {blendOption=Linear}, boneToModify {boneName=dyn_coat_l_02}, scale=<0.1 0.1 0.1>}
   n11 = "Transform (Modify) Bone - Bone: dyn_coat_fr_l_02"(ComponentPose=@n10, Alpha=@n1, Translation="0,0,0", Rotation="0,0,0", Scale="0.1,0.1,0.1") {scaleMode=BMM_Replace, alphaBoolBlend {blendOption=Linear}, boneToModify {boneName=dyn_coat_fr_l_02}, scale=<0.1 0.1 0.1>}
   n12 = "Transform (Modify) Bone - Bone: dyn_coat_bk_02"(ComponentPose=@n11, Alpha=@n1, Translation="0,0,0", Rotation="0,0,0", Scale="0.1,0.1,0.1") {scaleMode=BMM_Replace, alphaBoolBlend {blendOption=Linear}, boneToModify {boneName=dyn_coat_bk_02}, scale=<0.1 0.1 0.1>}
   n13 = "Transform (Modify) Bone - Bone: dyn_coat_bk_l_02"(ComponentPose=@n12, Alpha=@n1, Translation="0,0,0", Rotation="0,0,0", Scale="0.1,0.1,0.1") {scaleMode=BMM_Replace, alphaBoolBlend {blendOption=Linear}, boneToModify {boneName=dyn_coat_bk_l_02}, scale=<0.1 0.1 0.1>}
   n14 = "Transform (Modify) Bone - Bone: dyn_coat_bk_r_02"(ComponentPose=@n13, Alpha=@n1, Translation="0,0,0", Rotation="0,0,0", Scale="0.1,0.1,0.1") {scaleMode=BMM_Replace, alphaBoolBlend {blendOption=Linear}, boneToModify {boneName=dyn_coat_bk_r_02}, scale=<0.1 0.1 0.1>}
   n15 = "Transform (Modify) Bone - Bone: dyn_coat_fr_r_02"(ComponentPose=@n14, Alpha=@n1, Translation="0,0,0", Rotation="0,0,0", Scale="0.1,0.1,0.1") {scaleMode=BMM_Replace, alphaBoolBlend {blendOption=Linear}, boneToModify {boneName=dyn_coat_fr_r_02}, scale=<0.1 0.1 0.1>}
   n16 = "Transform (Modify) Bone - Bone: dyn_coat_r_02"(ComponentPose=@n15, Alpha=@n1, Translation="0,0,0", Rotation="0,0,0", Scale="0.1,0.1,0.1") {scaleMode=BMM_Replace, alphaBoolBlend {blendOption=Linear}, boneToModify {boneName=dyn_coat_r_02}, scale=<0.1 0.1 0.1>}
   n17 = RigidBody!(
    ComponentPose=@n16
    Alpha=Get in vehicle rigid alpha() -> (in vehicle rigid alpha:double)
    OverrideWorldGravity:Vector=Get gravity_override() -> (gravity_override:Vector)
    ExternalForce=Get vector_force() -> (vector_force:Vector)
   ) -> (Pose) {
    bClampLinearTranslationLimitToRefPose=true
    bOverrideWorldGravity=true
    overridePhysicsAsset=&/M_MED_Bad_Egg_Coat_Physics
    simulationSpace=BaseBoneSpace
    alphaBoolBlend {blendOption=Linear}
    baseBoneRef {boneName=pelvis}
    componentAppliedLinearAccClamp=<3000 3000 3000>
    componentLinearAccScale=<0.4 0.3 0.1>
    componentLinearVelScale=<0.4 0.3 0.1>
    overrideWorldGravity {z=-980}
   }
   n18 = RigidBody!(ComponentPose=@n17, Alpha=0.75, ExternalForce) -> (Pose) {
    bClampLinearTranslationLimitToRefPose=true
    bOverrideWorldGravity=true
    overridePhysicsAsset=&/M_MED_Bad_Egg_Collar_Physics
    simulationSpace=BaseBoneSpace
    alphaBoolBlend {blendOption=Linear}
    baseBoneRef {boneName=spine_01}
    componentAppliedLinearAccClamp=<3000 3000 1000>
    componentLinearAccScale=<0.5 0.4 0.3>
    componentLinearVelScale=<0.5 0.4 0.3>
    overrideWorldGravity {z=-980}
   }
   Output Pose(Result=Component To Local(ComponentPose=@n18)) {layerGroup=None}
  ]}
 EventGraph {nodes [
   Event BlueprintUpdateAnimation()
   > n1 = Branch(Condition=Get bIsSwimSprinting() -> (bIsSwimSprinting:bool))
   > Set gravity_override(gravity_override="0, 0,-200")
   n2 = Sequence!(execute=@n1.else) -> (then_0, then_1, then_2:Exec, then_3:Exec)
   n3 = Branch(execute=@n2.then_1, Condition=Get bIsOstrichVehicleDriver() -> (bIsOstrichVehicleDriver:bool))
   Set in vehicle rigid alpha(execute=[@n4.then, @n5.then, @n3.then], in vehicle rigid alpha:double=0) -> (Output_Get:double)
   > Set in vehicle scale alpha(in vehicle scale alpha:double=1) -> (Output_Get:double)
   n4 = Branch(execute=@n3.else, Condition=Get bIsFerretVehicleDriver() -> (bIsFerretVehicleDriver:bool))
   n5 = Branch(execute=@n4.else, Condition=Get bIsOctopusVehicleDriver() -> (bIsOctopusVehicleDriver:bool))
   Set in vehicle rigid alpha(execute=@n5.else, in vehicle rigid alpha=1)
   > Set in vehicle scale alpha(in vehicle scale alpha=0)
   n6 = Get Skel_Mesh()
   Set thigh_rot_r(
    execute=@n2.then_2
    thigh_rot_r=GetAlphaFromReferenceTransformDelta(SkeletalMeshComponent=@n6, BoneName=thigh_r, BaseName=pelvis, InputRange=(X=0,Y=-90), TwistAxis="0,0,1")
   )
   > Set thigh_rot_l(thigh_rot_l=GetAlphaFromReferenceTransformDelta(SkeletalMeshComponent=@n6, BoneName=thigh_l, BaseName=pelvis, InputRange=(X=-20,Y=-101), TwistAxis="0,0,1"))
   > Set Head_Rot_Back(Head_Rot_Back:double=GetAlphaFromReferenceTransformDelta(SkeletalMeshComponent=@n6, BoneName=head, BaseName=spine_05, InputRange=(X=0,Y=-45), TwistAxis="0,0,1")) -> (Output_Get:double)
   > Set arm_rot_fwd_r(arm_rot_fwd_r:double=GetAlphaFromReferenceTransformDelta(SkeletalMeshComponent=@n6, BoneName=upperarm_r, BaseName=clavicle_r, InputRange=(X=0,Y=60), TwistAxis="0,0,1")) -> (Output_Get:double)
   > Set arm_rot_high_r(arm_rot_high_r:double=GetAlphaFromReferenceTransformDelta(SkeletalMeshComponent=@n6, BoneName=upperarm_r, BaseName=clavicle_r, InputRange=(X=50,Y=90), TwistAxis="0,0,1")) -> (Output_Get:double)
   n7 = Branch(execute=@n2.then_3, Condition=Get bIsSkyDiving())
   > n8 = Branch(Condition=Get bIsParachuteOpened())
   > Set gravity_override(gravity_override="0, 0,-800")
   n9 = Break Fort Anim Input Skydiving External Force(FortAnimInput_SkydivingExternalForce=Get SkydivingExternalForceInput()) -> (
    bUseSkydivingVectorForce=~
    SkydivingVectorForce=~
    HeadToPelvisDirection=~
    FloatingMultiplier=~
    FloatingAdditive=~
    DivingMultiplier=~
    DivingAdditive=~
    ParachutingMultiplier=~
    ParachutingAdditive=~
   )
   > Set vector_force(vector_force:Vector=@n9.SkydivingVectorForce) -> (Output_Get:Vector)
   n10 = Branch(execute=@n8.else, Condition="float > float"(A=Get PawnVelocityZ(), B=50) {is=title K2Node_CallFunction})
   > Set gravity_override(gravity_override="0, 0,-300")
   > Set vector_force(vector_force=@n9.SkydivingVectorForce)
   Set gravity_override(execute=@n10.else, gravity_override="0, 0,-300")
   > Set vector_force(vector_force=@n9.SkydivingVectorForce)
   > Branch(Condition=Get bCanPlayCustomAnimations() -> (bCanPlayCustomAnimations:bool))
   > Set gravity_override(gravity_override="0, 0,150")
   Set gravity_override(execute=@n7.else, gravity_override="0, 0,-980")
   > Set vector_force(vector_force="0, 0, 0")
   Event BlueprintInitializeAnimation()
   > Set Skel_Mesh(Skel_Mesh=GetOwnerSkeletalMeshComponent())
   Event OnEndCrouching()
   > Set crouching alpha(crouching alpha:double=0) -> (Output_Get:double)
   Event OnBeginCrouching()
   > Set crouching alpha(crouching alpha=1)
   n11 = Branch(Condition=Get bIsSkyDiving())
   > n12 = Branch(Condition=Get bIsParachuteOpened())
   Set gravity_override(execute=[@n12.then, @n11.else], gravity_override="0, 0,-980")
   Set gravity_override(execute=@n12.else, gravity_override="0, 0,-300")
   n13 = Branch(Condition=Get bIsJumping())
   Set jump_alpha(execute=[@n13.then, @n14.then], jump_alpha:double=1) -> (Output_Get:double)
   n14 = Branch(execute=@n13.else, Condition=Get bIsFalling())
   Set jump_alpha(execute=@n14.else, jump_alpha=0)
  ]}
]
variables [
 arm_rot_fwd_r {type=number}
 arm_rot_high_r {type=number}
 crouching alpha {type=number}
 gravity_override {type=Vector}
 Head_Rot_Back {type=number}
 in vehicle rigid alpha {type=number}
 in vehicle scale alpha {type=number}
 jump_alpha {type=number}
 Skel_Mesh {type=SkeletalMeshComponent}
 thigh_rot_l {type=number}
 thigh_rot_r {type=number}
 vector_force {type=Vector}
]
