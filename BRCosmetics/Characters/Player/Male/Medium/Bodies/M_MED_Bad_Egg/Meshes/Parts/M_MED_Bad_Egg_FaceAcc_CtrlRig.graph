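# Control Rig for the Bad Egg medium male face accessory that shifts the glasses bone in response to inner-cheek movement.
# It reads the Z translation of the left and right inner cheek bones relative to faceAttach and remaps each to a clamped weight.
# The larger weight drives an additive local transform on dyn_glasses, applying a small offset and tilt.
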ControlRigBlueprint /BRCosmetics/Characters/Player/Male/Medium/Bodies/M_MED_Bad_Egg/Meshes/Parts/M_MED_Bad_Egg_FaceAcc_CtrlRig
graphs [
 RigVMFunctionLibrary {links [], nodes []}
 RigVMModel {nodes [
   n1 = &RigUnit_BeginExecution()
   n2 = &RigUnit_GetRelativeBoneTransform(Bone=L_cheek_inner, Space=faceAttach)
   n3 = &RigUnit_GetRelativeBoneTransform(Bone=R_cheek_inner, Space=faceAttach)
   /Script/ControlRig.RigUnit_ModifyBoneTransforms(
    BoneToModify=((Bone="dyn_glasses",Transform=(Rotation=(X=0,Y=0,Z=0,W=1),Translation=(X=0,Y=0.5,Z=0.5),Scale3D=(X=1,Y=1,Z=1))))
    Mode=AdditiveLocal
    WeightMaximum=1
    WeightMinimum=0
    ExecutePin=@n1.ExecutePin
    Weight=/Script/RigVM.RigVMFunction_MathFloatMax(
     A=&RigVMFunction_MathFloatRemap(SourceMaximum=5, SourceMinimum=3.86, TargetMinimum=0, bClamp=true, Value=@n2.Transform.Translation.Z).Result
     B=&RigVMFunction_MathFloatRemap(SourceMaximum=5, SourceMinimum=3.86, TargetMinimum=0, bClamp=true, Value=@n3.Transform.Translation.Z).Result
    ) {is=struct RigVMUnitNode, template="Maximum::Execute(in A,in B,out Result)"}.Result
    BoneToModify.0.Transform.Rotation=/Script/RigVM.RigVMFunction_MathQuaternionFromRotator(Rotator=(Pitch=0,Yaw=-5,Roll=3)) {is=struct RigVMUnitNode}.Result
   ) {is=struct RigVMUnitNode}
  ]}
]
hierarchy [
 root {loc=<0 0 0>, rot=<0 0 0 1>}
 pelvis {parent=root, loc=<0 -0.7071 93.1191>, rot=<0.0127 0.707 -0.0127 -0.707>}
 spine_01 {parent=pelvis, loc=<6.7494 -0.6388 0>, rot=<0 0 0.034 -0.9994>}
 spine_02 {parent=spine_01, loc=<7.2518 -0.035 0>, rot=<0 0 0.0238 -0.9997>}
 spine_03 {parent=spine_02, loc=<6.0104 0 0>, rot=<0 0 0.0149 -0.9999>}
 spine_04 {parent=spine_03, loc=<8.1024 0 0>, rot=<0 0 -0.0132 -0.9999>}
 spine_05 {parent=spine_04, loc=<10.0146 0 0>, rot=<0 0 -0.0015 -1>}
 neck_01 {parent=spine_05, loc=<11.3138 0.1529 0>, rot=<0 0 -0.2003 -0.9797>}
 neck_02 {parent=neck_01, loc=<4.5627 0 0>, rot=<0 0 0.0504 -0.9987>}
 head {parent=neck_02, loc=<5.3914 0 0>, rot=<0 0 0.115 -0.9934>}
 dyn_glasses {parent=head, loc=<4.9646 2.5213 0>, rot=<0 0 0 -1>}
 faceAttach {parent=head, loc=<-0.8266 -0.1832 0>, rot=<0.0044 -0.3827 -0.0044 -0.9239>, scale=<0.89 0.89 0.89>}
 L_cheek_inner {parent=faceAttach, loc=<4.2785 8.761 -1.4435>, rot=<0.0454 -0.6262 0.3191 -0.7099>}
 R_cheek_inner {parent=faceAttach, loc=<-1.4992 8.7305 4.3237>, rot=<-0.1979 -0.0545 -0.2598 -0.9436>}
]
